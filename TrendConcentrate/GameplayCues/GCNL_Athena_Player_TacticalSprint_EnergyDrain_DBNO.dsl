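; Looping gameplay cue that drives player audio for tactical sprint energy drain in the DBNO (downed) state.
; On start, for a locally viewed target only, it caches the TacticalSprintEnergyComponent and starts a repeating timer.
; Each tick feeds the energy percentage to the cue's audio components and sets DBNOSprintCheck true on the target's AudioParameterComponent.
; On removal, it sets that flag back to false for the local viewer.
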(event EventOnLoopingStart(EitherVFX-preferthisversion)
  (params (MyTarget :type Actor%20Object%20Reference) (Parameters :type Gameplay%20Cue%20Parameters%20Structure%20%28by%20ref%29) (ParticleComponents :type Array%20of%20FXSystem%20Component%20Object%20References) (AudioComponents :type Array%20of%20Audio%20Component%20Object%20References))
  (bind _target (Variables|Default|GetTarget))
  (bind _returnvalue (Actor|GetComponentbyClass _target "/TacticalSprintGame/Gameplay/TacticalSprintEnergyComponent.TacticalSprintEnergyComponent_C"))
  (Networking|CanExecuteCosmeticEvents
    (:True
      (Variables|Default|SetAudio%20Components AudioComponents)
      (Variables|Default|SetTarget MyTarget)
      (Utilities|Time|ClearandInvalidateTimerbyHandle 0 (Variables|Default|GetStepReset))
      (Utilities|IsLocallyViewed MyTarget
        (:Is%20Locally%20Viewed
          (Utilities|IsValid _returnvalue
            (:Is%20Valid
              (Variables|Default|SetTacSprintComp _returnvalue)
              (Utilities|Time|SetTimerbyEvent (EventDispatchers|CreateEvent :!event Custom|UpdateFX) (Variables|Default|GetFrequencyToCheckEnergy) true) ; Only operates for local players.
            )))))))

(event Custom|UpdateFX
  ;[1 1 1] Updates Energy Percentage variable in associated audio component.  This only needs to run for locally controlled or spectated pawns.
  (Variables|Default|SetCurrentEnergyPercentage (Energy|GetCurrentEnergyPercentage (Variables|Default|GetTacSprintComp)))
  (for _array_element (Variables|Default|GetAudio%20Components)
    (Audio|Parameter|SetFloatParameter _array_element "EnergyPercentage" (Variables|Default|GetCurrentEnergyPercentage) :!class /Script/AudioExtensions.AudioParameterControllerInterface))
  (Audio|Parameter|SetBooleanParameter (Actor|GetComponentbyClass (Variables|Default|GetTarget) "/Script/AudioGameplay.AudioParameterComponent") "DBNOSprintCheck" true :!class /Script/AudioExtensions.AudioParameterControllerInterface))

(event EventOnRemoval(EitherVFX-preferthisversion)
  (params (MyTarget :type Actor%20Object%20Reference) (Parameters :type Gameplay%20Cue%20Parameters%20Structure%20%28by%20ref%29) (ParticleComponents :type Array%20of%20FXSystem%20Component%20Object%20References) (AudioComponents :type Array%20of%20Audio%20Component%20Object%20References) (BurstCameraShakeInstance :type Camera%20Shake%20Base%20Object%20Reference) (BurstDecalInstance :type Decal%20Actor%20Object%20Reference))
  (Networking|CanExecuteCosmeticEvents
    (:True
      (Utilities|IsLocallyViewed MyTarget
        (:Is%20Locally%20Viewed
          (Audio|Parameter|SetBooleanParameter (Actor|GetComponentbyClass (Variables|Default|GetTarget) "/Script/AudioGameplay.AudioParameterComponent") "DBNOSprintCheck" false :!class /Script/AudioExtensions.AudioParameterControllerInterface))))))

(fn ConstructionScript)
