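; Looping gameplay cue that plays audio and repeating local force feedback while a downed (DBNO) player pawn ragdolls.
; On start it attaches a first- or third-person loop, sets clothing foley from the character cosmetic's gameplay tag, and starts a rumble timer.
; Surface-tracking changes swap the rolling waves through data table lookups with a Default fallback.
; Removal clears the timer, fades the loop and plays a surface-based finishing sound.
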(event EventBeginPlay)

(event Collision|EventActorBeginOverlap
  (params (OtherActor :type Actor%20Object%20Reference)))

(event EventTick
  (params (DeltaSeconds :type Float%20%28single-precision%29)))

(event EventOnLoopingStart(EitherVFX-preferthisversion)
  (params (MyTarget :type Actor%20Object%20Reference) (Parameters :type Gameplay%20Cue%20Parameters%20Structure%20%28by%20ref%29) (ParticleComponents :type Array%20of%20FXSystem%20Component%20Object%20References) (AudioComponents :type Array%20of%20Audio%20Component%20Object%20References))
  (|CallParentFunction MyTarget Parameters ParticleComponents AudioComponents)
  (bind _asfort_player_pawn (Utilities|Casting|CastToFortPlayerPawn MyTarget))
  (Variables|Default|SetPlayerPawn _asfort_player_pawn)
  (CallFunction|StartAudio)
  (EventDispatchers|BindEventtoOnSurfaceTypeChanged (Footsteps|GetSurfaceTrackingComponent (Variables|Default|GetPlayerPawn)) (EventDispatchers|CreateEvent :!event Custom|OnSurfaceChanged))
  (bind _returnvalue (Utilities|Time|SetTimerbyEvent (delegate Custom|CustomEvent) 0.35 true))
  (Variables|Default|SetFF_Timer _returnvalue))

(event Custom|CustomEvent
  (Fortnite|PlayLocalForceFeedbackatLocation (stale EventOnLoopingStart(EitherVFX-preferthisversion) MyTarget) "/TrendConcentrate/Assets/ForceFeedback/FF_Ragdoll.FF_Ragdoll" (Transformation|GetActorLocation (Variables|Default|GetPlayerPawn)) 1.0))

(event Custom|OnSurfaceChanged
  (params (SurfaceActor :type Actor%20Object%20Reference) (CurrentSurfaceType :type EPhysicalSurface%20Enum :default "SurfaceType_Default") (PrevSurfaceType :type EPhysicalSurface%20Enum :default "SurfaceType_Default"))
  (CallFunction|UpdateSurfaceAudio CurrentSurfaceType))

(event EventOnRemoval(EitherVFX-preferthisversion)
  (params (MyTarget :type Actor%20Object%20Reference) (Parameters :type Gameplay%20Cue%20Parameters%20Structure%20%28by%20ref%29) (ParticleComponents :type Array%20of%20FXSystem%20Component%20Object%20References) (AudioComponents :type Array%20of%20Audio%20Component%20Object%20References) (BurstCameraShakeInstance :type Camera%20Shake%20Base%20Object%20Reference) (BurstDecalInstance :type Decal%20Actor%20Object%20Reference))
  (bind (_surface_5_48cf6667487fc809d049d8b9ba4415d9 _fallingsounds_6_0bd154fb44c999f60e7a9baee8e4fbdc) (Utilities|Struct|BreakStructDBNORagDoll (Variables|Default|GetEndingStruct)))
  (bind _finishwavetable (Variables|Audio|GetFinishWaveTable))
  (|CallParentFunction MyTarget Parameters ParticleComponents AudioComponents BurstCameraShakeInstance BurstDecalInstance)
  (Utilities|Time|ClearandInvalidateTimerbyHandle MyTarget (Variables|Default|GetFF_Timer))
  (Audio|Components|Audio|FadeOut (Variables|Default|GetDbnoAudioLoop) 0.5 0.0 "Linear" :!class /Script/Engine.AudioComponent)
  (bind _returnvalue (Audio|SpawnSoundatLocation MyTarget (Variables|Default|GetDBNORagDollEnd) (Transformation|GetActorLocation MyTarget)))
  (bind _returnvalue_2 (Utilities|GetDataTableRow _finishwavetable (Fortnite|Utility|GetPhysicalSurfaceName (Class|FortSurfaceTrackingComponent|GetCurrentSurfaceType (Footsteps|GetSurfaceTrackingComponent (Variables|Default|GetPlayerPawn)))))
    (:then
      (Variables|Default|SetEndingStruct _returnvalue_2))
    (:RowNotFound
      (bind _returnvalue_3 (Utilities|GetDataTableRow _finishwavetable "Default"))
      (Variables|Default|SetEndingStruct _returnvalue_3)))
  (Audio|Parameter|SetObjectArrayParameter _returnvalue "FinishSounds" _fallingsounds_6_0bd154fb44c999f60e7a9baee8e4fbdc)
  (Audio|Parameter|ExecuteTriggerParameter _returnvalue "TriggerAudio"))

(fn SetClothAudio
  (bind (_soundlibrarytag_11_2a36fa054cbaadb349620e9404112081 _clothsounds_7_cdb0f69d49f9c41dfc89308f55ff24f4) (Utilities|Struct|BreakStructDBNORagdollClothingType (Variables|Default|GetClothingStruct)))
  (bind _playerpawn (Variables|Default|GetPlayerPawn))
  (bind _returnvalue (CosmeticLoadout|GetItemDefinitionfromLoadoutwithActor _playerpawn "(TagName=\"LoadoutSlot.Character\")"))
  (bind _returnvalue_2 (GameplayTags|MakeGameplayTagContainerfromTag "(TagName=\"SoundLibrary.ID.ClothingType\")"))
  (Utilities|IsValid _returnvalue
    (:Is%20Valid
      (bind _itemdata (Class|ItemDefinitionBase|GetItemData _returnvalue)
        (:Found
          (if (Utilities|Array|IsValidIndex (GameplayTags|BreakGameplayTagContainer (GameplayTags|Filter (Utilities|Struct|BreakOwnedGameplayTagsItemData _itemdata) _returnvalue_2)))
            (bind _returnvalue_3 (Utilities|GetDataTableRow (Variables|Audio|GetClothingTable) (Utilities|String|StringToName (GameplayTags|GetDebugStringfromGameplayTag (Utilities|Array|Get(acopy) (GameplayTags|BreakGameplayTagContainer (GameplayTags|Filter (Utilities|Struct|BreakOwnedGameplayTagsItemData _itemdata) _returnvalue_2))))))
              (:then
                (Variables|Default|SetClothingStruct _returnvalue_3))
              (:RowNotFound
                (bind _returnvalue_4 (Utilities|GetDataTableRow (Variables|Audio|GetClothingTable) "SoundLibrary.ID.ClothingType.Generic"))
                (Variables|Default|SetClothingStruct _returnvalue_4)))
            (bind _dbnoaudioloop (Variables|Default|GetDbnoAudioLoop)
              (:then
                (Audio|Parameter|SetObjectArrayParameter _dbnoaudioloop "CharacterFoley" _clothsounds_7_cdb0f69d49f9c41dfc89308f55ff24f4)
                (return))
              (:else))
            (else
              _)))
        (:NotFound
          _)))))

(fn StartAudio
  (bind _returnvalue (Audio|SpawnSoundAttached (select (Pawn|IsLocallyViewed (Variables|Default|GetPlayerPawn)) (Variables|Audio|Get1pLoop) (Variables|Audio|Get3pLoop)) (Variables|/Script/Engine.Actor|GetRootComponent (Variables|Default|GetPlayerPawn))))
  (Utilities|IsValid _returnvalue
    (:Is%20Valid
      (Variables|Default|SetDbnoAudioLoop _returnvalue)
      (CallFunction|SetClothAudio)
      (CallFunction|UpdateSurfaceAudio (Class|FortSurfaceTrackingComponent|GetCurrentSurfaceType (Footsteps|GetSurfaceTrackingComponent (Variables|Default|GetPlayerPawn))))
      (Audio|Parameter|ExecuteTriggerParameter (Variables|Default|GetDbnoAudioLoop) "StartRoll"))))

(fn UpdateSurfaceAudio
  (params (PhysicalSurface :type EPhysicalSurface%20Enum :default "SurfaceType_Default"))
  (bind (_surface_5_48cf6667487fc809d049d8b9ba4415d9 _fallingsounds_6_0bd154fb44c999f60e7a9baee8e4fbdc) (Utilities|Struct|BreakStructDBNORagDoll (LocalVariables|GetSurfaceSelection)))
  (bind _loopwavtable (Variables|Audio|GetLoopWavTable))
  (bind _returnvalue (Utilities|GetDataTableRow _loopwavtable (Fortnite|Utility|GetPhysicalSurfaceName PhysicalSurface))
    (:then
      (LocalVariables|SetSurfaceSelection _returnvalue))
    (:RowNotFound
      (bind _returnvalue_2 (Utilities|GetDataTableRow _loopwavtable "Default"))
      (LocalVariables|SetSurfaceSelection _returnvalue_2)))
  (bind _dbnoaudioloop (Variables|Default|GetDbnoAudioLoop)
    (:then
      (Audio|Parameter|SetObjectArrayParameter _dbnoaudioloop "RollingLoop.Waves" _fallingsounds_6_0bd154fb44c999f60e7a9baee8e4fbdc)
      (Audio|Parameter|ExecuteTriggerParameter _dbnoaudioloop "AssetSwitch"))
    (:else)))

(fn ConstructionScript)
