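; Gameplay cue notify that, going by its name, signals the end of the DBNO build points cooldown.
; OnExecute plays a generic Creative toy activation sound as non-positional 2D audio, using the target actor as world context, then returns false.
; The BeginPlay, overlap, Tick and EndPlay events are empty, so the cue has no other behavior.
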(event EventBeginPlay)

(event Collision|EventActorBeginOverlap
  (params (OtherActor :type Actor%20Object%20Reference)))

(event EventTick
  (params (DeltaSeconds :type Float%20%28single-precision%29)))

(event EventEndPlay
  (params (EndPlayReason :type EEndPlayReason%20Enum :default "Destroyed")))

(fn OnExecute
  (params (MyTarget :type Actor%20Object%20Reference) (Parameters :type Gameplay%20Cue%20Parameters%20Structure%20%28by%20ref%29))
  (Audio|PlaySound2D MyTarget "/Game/Sounds/Creative/Toys/Generic/Toy_Generic_Activate_Cue.Toy_Generic_Activate_Cue")
  (return false))

(fn ConstructionScript)
