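; Burst gameplay cue for the DBNO lunge start that plays surface-aware roll foley on the target Fort pawn.
; OnBurst starts the sound through the Sound Library, applies the current surface and binds to the pawn's surface tracking changes.
; Surface updates look up a data table row, falling back to Default, then send its sounds to the audio component and fire a switch trigger.
; When the audio finishes, the surface-change listener is unbound.
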(event EventBeginPlay)

(event Collision|EventActorBeginOverlap
  (params (OtherActor :type Actor%20Object%20Reference)))

(event EventTick
  (params (DeltaSeconds :type Float%20%28single-precision%29)))

(event EventEndPlay
  (params (EndPlayReason :type EEndPlayReason%20Enum :default "Destroyed")))

(event Custom|AudioFin
  (EventDispatchers|UnbindEventfromOnSurfaceTypeChanged (Footsteps|GetSurfaceTrackingComponent (Variables|Default|GetAs%20Fort%20Player%20Pawn)) (EventDispatchers|CreateEvent :!event SurfaceChange)))

(fn OnBurst
  (params (MyTarget :type Actor%20Object%20Reference) (Parameters :type Gameplay%20Cue%20Parameters%20Structure%20%28by%20ref%29))
  (bind _returnvalue (|CallParentFunction MyTarget Parameters))
  (bind _asfort_player_pawn (Utilities|Casting|CastToFortPlayerPawn MyTarget))
  (bind _output_get (Variables|Default|SetAs%20Fort%20Player%20Pawn _asfort_player_pawn))
  (CallFunction|StartAudio)
  (CallFunction|UpdateSurfaceAudio (Class|FortSurfaceTrackingComponent|GetCurrentSurfaceType (Footsteps|GetSurfaceTrackingComponent _output_get)))
  (EventDispatchers|BindEventtoOnSurfaceTypeChanged (Footsteps|GetSurfaceTrackingComponent _output_get) (EventDispatchers|CreateEvent :!event SurfaceChange))
  (return _returnvalue))

(fn StartAudio
  (bind _audiocomp (Variables|Default|GetAudioComp)
    (:then
      (EventDispatchers|UnbindEventfromOnAudioFinished _audiocomp (EventDispatchers|CreateEvent :!event AudioFin)))
    (:else
      _))
  (bind _returnvalue (SoundLibrary|PlaySoundSimple (WorldSubsystems|GetSoundLibrarySubsystem (Variables|Default|GetAs%20Fort%20Player%20Pawn)) (Variables|Default|GetAs%20Fort%20Player%20Pawn) (Utilities|Struct|MakeSimpleSoundSettings "(TagName=\"SoundLibrary.Foley.DBNO.Roll\")" true false true false 1.0 1.0 0.0 0 0 "None" (Variables|/Script/Engine.Actor|GetRootComponent (Variables|Default|GetAs%20Fort%20Player%20Pawn))) :!class /Script/SoundLibrary.SoundLibrarySubsystem))
  (if (Utilities|Array|IsValidIndex (.AudioComponents (Utilities|Struct|BreakSoundLibraryPlayResult _returnvalue)))
    (bind _output_get (Variables|Default|SetAudioComp (Utilities|Array|Get(acopy) (.AudioComponents (Utilities|Struct|BreakSoundLibraryPlayResult _returnvalue)))))
    (EventDispatchers|BindEventtoOnAudioFinished _output_get (EventDispatchers|CreateEvent :!event AudioFin))
    (else
      _))
  (return))

(fn SurfaceChange
  (params (SurfaceActor :type Actor%20Object%20Reference) (CurrentSurfaceType :type EPhysicalSurface%20Enum :default "SurfaceType_Default") (PrevSurfaceType :type EPhysicalSurface%20Enum :default "SurfaceType_Default"))
  (if (not (Utilities|Enum|Equal(Enum) CurrentSurfaceType (Variables|Default|GetCurrent%20Surface%20Type)))
    (bind _output_get (Variables|Default|SetCurrent%20Surface%20Type CurrentSurfaceType))
    (CallFunction|UpdateSurfaceAudio _output_get)))

(fn UpdateSurfaceAudio
  (params (SurfaceType :type EPhysicalSurface%20Enum :default "SurfaceType_Default"))
  (bind (_surface_5_48cf6667487fc809d049d8b9ba4415d9 _fallingsounds_6_0bd154fb44c999f60e7a9baee8e4fbdc) (Utilities|Struct|BreakStructDBNORagDoll (LocalVariables|GetData)))
  (bind _datatable (Variables|Default|GetDataTable))
  (bind _audiocomp (Variables|Default|GetAudioComp))
  (Variables|Default|SetCurrent%20Surface%20Type SurfaceType)
  (Variables|Default|GetAudioComp
    (:then
      (bind _returnvalue (Utilities|GetDataTableRow _datatable (Fortnite|Utility|GetPhysicalSurfaceName (Variables|GetSurfaceType)))
        (:then
          (LocalVariables|SetData _returnvalue))
        (:RowNotFound
          (bind _returnvalue_2 (Utilities|GetDataTableRow _datatable "Default"))
          (LocalVariables|SetData _returnvalue_2)))
      (Audio|Parameter|SetObjectArrayParameter _audiocomp "SurfaceSounds" _fallingsounds_6_0bd154fb44c999f60e7a9baee8e4fbdc)
      (Audio|Parameter|ExecuteTriggerParameter _audiocomp "TriggerSwitch"))
    (:else)))

(fn ConstructionScript)
